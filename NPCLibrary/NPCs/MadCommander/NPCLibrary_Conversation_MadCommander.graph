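# Conversation database for the Mad Commander NPC (Ruckus), defining what he says when a player starts talking to him.
# A single entry point leads to a greeting speech node that picks lines by priority for first-ever meetings, per-match conversation count, or specific player outfits, with a default fallback.
# The speech then branches through link nodes into the shared NPC quests and basic services conversations, plus his own dialogue entry.
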ConversationDatabase /NPCLibrary/NPCs/MadCommander/NPCLibrary_Conversation_MadCommander
entry_points [{tag=Conversation.Entry.MadCommander, destinations [#1]}]
nodes [
 ConversationEntryPointNode_7 {
  id=#1
  class=ConversationEntryPointNode
  outputs [#2]
  properties {entryTag {tagName=Conversation.Entry.MadCommander}}
 }
 Speech_C_4 {
  id=#2
  class=Speech_C
  outputs [#3, #4, #5]
  properties {
   nodeName=""
   generalConfig {
    defaultMessage=Who needs leader when Ruckus can be leader AND follower?
    contextualMessages [
     {
      message="I am Ruckus!! THE LONE RABBLER!!!"
      contextRequirements [{requirement {class=FortControllerRequirement_IsFirstEverConversationWithNPC}}]
      priority {curve {rowName=Tandem.Speech.Priority.FirstEverInteraction}}
     }
     {
      message=Who needs leader when Ruckus can be leader AND follower?
      contextRequirements [{requirement {
         class=FortControllerRequirement_CheckNPCMatchConversationCount
         properties {nPCUniqueIdTag {tagName=AISpawnerData.Type.Specialist.Heavy.MadCommander}}
        }}]
      priority {curve {rowName=None}}
     }
     {
      message="Cause Ruckus! Rouse rabble! BE FREE!!"
      contextRequirements [{requirement {
         class=FortControllerRequirement_CheckNPCMatchConversationCount
         properties {nPCUniqueIdTag {tagName=AISpawnerData.Type.Specialist.Heavy.MadCommander}}
        }}]
      priority {curve {rowName=None}}
     }
     {
      message="Ruckus the Rabbler follows his own path now!"
      contextRequirements [{requirement {
         class=FortControllerRequirement_CheckNPCMatchConversationCount
         properties {nPCUniqueIdTag {tagName=AISpawnerData.Type.Specialist.Heavy.MadCommander}}
        }}]
      priority {curve {rowName=None}}
     }
     {
      message=The Chums disbanded without telling Ruckus...
      contextRequirements [{requirement {
         class=FortControllerRequirement_HasCID
         properties {allowedCIds [/BRCosmetics/Athena/Items/Cosmetics/Characters/CID_A_097_Athena_Commando_F_WastelandWarrior]}
        }}]
      priority {curve {rowName=Tandem.Speech.Priority.OutfitReaction}}
     }
     {
      message="Without the Chums, Ruckus stands alone..."
      contextRequirements [{requirement {
         class=FortControllerRequirement_HasCID
         properties {allowedCIds [/BRCosmetics/Athena/Items/Cosmetics/Characters/CID_097_Athena_Commando_F_RockerPunk]}
        }}]
      priority {curve {rowName=Tandem.Speech.Priority.OutfitReaction}}
     }
     {
      message=Ruckus yearns for his Chums... but he must be alone...
      contextRequirements [{requirement {
         class=FortControllerRequirement_HasCID
         properties {allowedCIds [/BRCosmetics/Athena/Items/Cosmetics/Characters/CID_196_Athena_Commando_M_Biker]}
        }}]
      priority {curve {rowName=Tandem.Speech.Priority.OutfitReaction}}
     }
    ]
   }
  }
 }
 ConversationLinkNode_2 {
  id=#5
  class=ConversationLinkNode
  properties {remoteEntryTag {tagName=Conversation.Entry.MadCommander.Dialogue}}
 }
 ConversationLinkNode_0 {id=#4, class=ConversationLinkNode, properties {remoteEntryTag {tagName=Conversation.Entry.BasicServices}}}
 ConversationLinkNode_1 {id=#3, class=ConversationLinkNode, properties {remoteEntryTag {tagName=Conversation.Entry.NPCLibrary.Quests}}}
]
