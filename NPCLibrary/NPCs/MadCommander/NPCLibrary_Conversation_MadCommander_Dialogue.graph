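# Conversation database for the Mad Commander NPC's dialogue, entered by the Conversation.Entry.MadCommander.Dialogue tag.
# An opening "Hmm?" speech node, shown as an "Ask about..." choice, branches to a "Ruckus the Rabbler" topic or a Back node.
# The topic plays a linear chain of lines about Ruckus being abandoned by Chiara's Chums, ending in a hint to walk together.
# A ReturnToStart node with a BACK choice closes the chain.
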ConversationDatabase /NPCLibrary/NPCs/MadCommander/NPCLibrary_Conversation_MadCommander_Dialogue
entry_points [{tag=Conversation.Entry.MadCommander.Dialogue, destinations [#1]}]
nodes [
 ConversationEntryPointNode_0 {
  id=#1
  class=ConversationEntryPointNode
  outputs [#2]
  properties {entryTag {tagName=Conversation.Entry.MadCommander.Dialogue}}
 }
 Speech_C_0 {
  id=#2
  class=Speech_C
  outputs [#3, #4]
  properties {nodeName="", softTaskIcon=&T_UI_NPC_question, generalConfig {defaultMessage=Hmm?}}
  sub_nodes [ConversationChoiceNode_5 {class=ConversationChoiceNode, properties {defaultChoiceDisplayText=Ask about...}}]
 }
 Back_C_0 {
  id=#4
  class=Back_C
  sub_nodes [ConversationChoiceNode_7 {class=ConversationChoiceNode, properties {defaultChoiceDisplayText=BACK}}]
 }
 Speech_C_7 {
  id=#3
  class=Speech_C
  outputs [#5]
  properties {
   nodeName=""
   softTaskIcon=&T_UI_NPC_question
   generalConfig {defaultMessage="Long ago, Ruckus belonged to Chiara's Chums!"}
  }
  sub_nodes [ConversationChoiceNode_0 {class=ConversationChoiceNode, properties {defaultChoiceDisplayText=Ruckus the Rabbler}}]
 }
 Speech_C_15 {
  id=#5
  class=Speech_C
  outputs [#6]
  properties {
   nodeName=""
   softTaskIcon=&T_UI_NPC_question
   generalConfig {defaultMessage=But the Chums disbanded. Without telling Ruckus.}
  }
 }
 Speech_C_14 {
  id=#6
  class=Speech_C
  outputs [#7]
  properties {
   nodeName=""
   softTaskIcon=&T_UI_NPC_question
   generalConfig {defaultMessage="Now Ruckus is RUCKUS THE RABBLER! And he walks ALONE!!"}
  }
 }
 Speech_C_1 {
  id=#7
  class=Speech_C
  outputs [#8]
  properties {
   nodeName=""
   softTaskIcon=&T_UI_NPC_question
   generalConfig {defaultMessage=Unless... you walk with Ruckus for a while...?}
  }
 }
 ReturnToStart_C_23 {
  id=#8
  class=ReturnToStart_C
  properties {softTaskIcon=&T_UI_NPC_Back}
  sub_nodes [ConversationChoiceNode_32 {class=ConversationChoiceNode, properties {defaultChoiceDisplayText=BACK}}]
 }
]
